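# Conversation database for the Wreck Raider NPC in the NPC library, opening with a shark-themed greeting.
# A single entry point leads to a speech node whose default line gives way to prioritized contextual lines for a first-ever conversation or the player wearing the Shark outfit.
# The speech then branches through link nodes into shared conversations for Season 16 Spire Joneses, NPC library quests and basic services.
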ConversationDatabase /NPCLibrary/NPCs/WreckRaider/NPCLibrary_Conversation_WreckRaider
entry_points [{tag=Conversation.Entry.WreckRaider, destinations [#1]}]
nodes [
 ConversationEntryPointNode_7 {
  id=#1
  class=ConversationEntryPointNode
  outputs [#2]
  properties {entryTag {tagName=Conversation.Entry.WreckRaider}}
 }
 Speech_C_4 {
  id=#2
  class=Speech_C
  outputs [#3, #4, #5]
  properties {
   nodeName=""
   generalConfig {
    defaultMessage=Seen any fins out there?
    contextualMessages [
     {
      message="Did you know there are sharks in the ocean? Because I did <b>not.</>"
      contextRequirements [{requirement {class=FortControllerRequirement_IsFirstEverConversationWithNPC}}]
      priority {curve {rowName=Tandem.Speech.Priority.FirstEverInteraction}}
     }
     {
      message="Not cool. "
      contextRequirements [{requirement {
         class=FortControllerRequirement_HasCID
         properties {allowedCIds [/BRCosmetics/Athena/Items/Cosmetics/Characters/CID_168_Athena_Commando_M_Shark]}
        }}]
      priority {curve {rowName=Tandem.Speech.Priority.OutfitReaction}}
     }
    ]
   }
  }
 }
 ConversationLinkNode_0 {id=#5, class=ConversationLinkNode, properties {remoteEntryTag {tagName=Conversation.Entry.BasicServices}}}
 ConversationLinkNode_1 {id=#4, class=ConversationLinkNode, properties {remoteEntryTag {tagName=Conversation.Entry.NPCLibrary.Quests}}}
 ConversationLinkNode_2 {
  id=#3
  class=ConversationLinkNode
  properties {remoteEntryTag {tagName=Conversation.Entry.Season16.TheSpire.Joneses}}
 }
]
